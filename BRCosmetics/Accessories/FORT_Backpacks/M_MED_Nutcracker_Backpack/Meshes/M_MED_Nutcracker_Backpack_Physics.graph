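# Physics asset for the Nutcracker backpack cosmetic, giving it a simulated body and two dangling legs.
# A box body on the torso joint and a box on each upper leg joint form the collision shapes.
# Each leg is constrained to the torso with limited swing and locked twist, with slightly different swing limits per side, so the legs can sway without spinning.
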PhysicsAsset /BRCosmetics/Accessories/FORT_Backpacks/M_MED_Nutcracker_Backpack/Meshes/M_MED_Nutcracker_Backpack_Physics
bodies {
 body_jnt [{shapeType=Box, extentX=7.1525, extentY=6.2674, extentZ=6.5029, center=<1.1524 0.082 -0.2645>, rotation=<rot -3.803 -90 14.1477>}]
 dyn_legA_l_jnt [{shapeType=Box, extentX=1.7325, extentY=1.7317, extentZ=2.8001, center=<1.292 -0.0023 0.0061>, rotation=<rot 0 90.8524 -89.102>}]
 dyn_legA_r_jnt [{shapeType=Box, extentX=1.7, extentY=1.6992, extentZ=2.819, center=<1.3009 -0.0079 -0.0362>, rotation=<rot 0 -89.1254 -89.3316>}]
}
constraints [
 {bone1Name=dyn_legA_r_jnt, bone2Name=body_jnt, swing1LimitDegrees=27.4286, swing2LimitDegrees=45, twistLimitDegrees=45, twistMotion=Locked}
 {
  bone1Name=dyn_legA_l_jnt
  bone2Name=body_jnt
  swing1LimitDegrees=26.1429
  swing2LimitDegrees=31.2857
  twistLimitDegrees=45
  twistMotion=Locked
 }
]
