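# Animation Blueprint for the Nutcracker backpack cosmetic, adding secondary motion to its bird-like body parts.
# Input pose passes through an offset, RigidBody, then chained Anim Dynamics for body, wing, tail, neck, legs, head, beak, hat and eyes.
# Copy Bone Delta mirrors lower beak rotation onto the upper beak; both upper-leg (legA) nodes are unconnected.
# Each update maps spine_05-to-spine_01 yaw delta into final additive body and root adjustments.
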AnimBlueprint /BRCosmetics/Accessories/FORT_Backpacks/M_MED_Nutcracker_Backpack/Meshes/M_MED_Nutcracker_Backpack_AnimBP
target_skeleton=&/M_MED_Nutcracker_Backpack_Skeleton
graphs [
 AnimGraph {nodes [
   Anim Dynamics - Bone: dyn_legA_l_jnt(Alpha=1, ExternalForce="0,0,0") {
    angularDampingOverride=150
    bGravityOverrideInSimSpace=true
    bUseGravityOverride=true
    alphaBoolBlend {blendOption=Linear}
    boundBone {boneName=dyn_legA_l_jnt}
    componentAppliedLinearAccClamp=<300 300 300>
    componentLinearAccScale=<1 1 1>
    componentLinearVelScale=<1 1 1>
    gravityOverride {z=-981}
    physicsBodyDefinitions [{
      sphereCollisionRadius=0
      boundBone {boneName=dyn_legA_l_jnt}
      boxExtents=<10 10 6>
      constraintSetup {angularLimitsMax=<0 33 44>, angularLimitsMin=<0 -33 0>, angularTarget=<0 0 0>}
      localJointOffset=<4 -0.1 -1.5>
     }]
   }
   Anim Dynamics - Bone: dyn_legA_r_jnt(Alpha=1, ExternalForce="0,0,0") {
    angularDampingOverride=150
    bGravityOverrideInSimSpace=true
    bUseGravityOverride=true
    alphaBoolBlend {blendOption=Linear}
    boundBone {boneName=dyn_legA_r_jnt}
    componentAppliedLinearAccClamp=<300 300 300>
    componentLinearAccScale=<1 1 1>
    componentLinearVelScale=<1 1 1>
    gravityOverride {z=-981}
    physicsBodyDefinitions [{
      sphereCollisionRadius=0
      boundBone {boneName=dyn_legA_r_jnt}
      boxExtents=<6 10 10>
      constraintSetup {angularTargetAxis=AxisY, angularLimitsMax=<0 33 12>, angularLimitsMin=<0 -33 -44>, angularTarget=<0 0 0>}
      localJointOffset=<3 0 -1>
     }]
   }
   AnimGraphNode_SubInput_0 = Input Pose()
   n1 = "Transform (Modify) Bone - Bone: offsetAll"!(ComponentPose=Local To Component(LocalPose=@AnimGraphNode_SubInput_0), Alpha=1, Rotation="0,5,0") -> (Pose) {rotationMode=BMM_Additive, alphaBoolBlend {blendOption=Linear}, boneToModify {boneName=offsetAll}, rotation {yaw=5}}
   n2 = RigidBody(ComponentPose=@n1) {
    bOverrideWorldGravity=true
    overridePhysicsAsset=&/M_MED_Nutcracker_Backpack_Physics
    alphaBoolBlend {blendOption=Linear}
    componentAppliedLinearAccClamp=<300 300 300>
    componentLinearAccScale=<1 1 1>
    componentLinearVelScale=<1 1 1>
    overrideWorldGravity {z=-980}
   }
   n3 = Anim Dynamics - Bone: body_jnt(ComponentPose=@n2, Alpha=1, ExternalForce="0,0,0") {
    bGravityOverrideInSimSpace=true
    bUseGravityOverride=true
    alphaBoolBlend {blendOption=Linear}
    boundBone {boneName=body_jnt}
    componentAppliedLinearAccClamp=<300 300 300>
    componentLinearAccScale=<1 1 1>
    componentLinearVelScale=<1 1 1>
    gravityOverride {z=-981}
    physicsBodyDefinitions [{
      sphereCollisionRadius=0
      boundBone {boneName=body_jnt}
      boxExtents=<10 10 3>
      constraintSetup {angularTargetAxis=AxisZ, twistAxis=AxisZ, angularLimitsMax=<5 3 0>, angularLimitsMin=<-1 -3 0>, angularTarget=<0 0 0>}
      localJointOffset=<0 0 -5>
     }]
   }
   n4 = Anim Dynamics - Bone: dyn_wing_r_jnt(ComponentPose=@n3, Alpha=1, ExternalForce="0,0,0") {
    bGravityOverrideInSimSpace=true
    bUseGravityOverride=true
    alphaBoolBlend {blendOption=Linear}
    boundBone {boneName=dyn_wing_r_jnt}
    componentAppliedLinearAccClamp=<300 300 300>
    componentLinearAccScale=<1 1 1>
    componentLinearVelScale=<1 1 1>
    gravityOverride {z=-981}
    physicsBodyDefinitions [{
      sphereCollisionRadius=0
      boundBone {boneName=dyn_wing_r_jnt}
      boxExtents=<10 10 3>
      constraintSetup {angularTargetAxis=AxisZ, twistAxis=AxisZ, angularLimitsMax=<12 0 0>, angularLimitsMin=<-12 0 0>, angularTarget=<0 0 0>}
      localJointOffset=<0 0 -5>
     }]
   }
   n5 = Anim Dynamics - Bone: dyn_tail_jnt(ComponentPose=@n4, Alpha=1, ExternalForce="0,0,0") {
    angularDampingOverride=50
    bGravityOverrideInSimSpace=true
    bUseGravityOverride=true
    alphaBoolBlend {blendOption=Linear}
    boundBone {boneName=dyn_tail_jnt}
    componentAppliedLinearAccClamp=<300 300 300>
    componentLinearAccScale=<1 1 1>
    componentLinearVelScale=<1 1 1>
    gravityOverride {z=-981}
    physicsBodyDefinitions [{
      sphereCollisionRadius=0
      boundBone {boneName=dyn_tail_jnt}
      boxExtents=<10 10 3>
      constraintSetup {angularLimitsMax=<0 12 0>, angularLimitsMin=<0 -12 0>, angularTarget=<0 0 0>}
      localJointOffset=<-4.5 0 -5>
     }]
   }
   n6 = Anim Dynamics - Bone: dyn_neck_jnt(ComponentPose=@n5, Alpha=1, ExternalForce="0,0,0") {
    angularDampingOverride=50
    bGravityOverrideInSimSpace=true
    bUseGravityOverride=true
    alphaBoolBlend {blendOption=Linear}
    boundBone {boneName=dyn_neck_jnt}
    componentAppliedLinearAccClamp=<300 300 300>
    componentLinearAccScale=<1 1 1>
    componentLinearVelScale=<1 1 1>
    gravityOverride {z=-981}
    physicsBodyDefinitions [{
      sphereCollisionRadius=0
      boundBone {boneName=dyn_neck_jnt}
      boxExtents=<40 20 20>
      constraintSetup {angularTargetAxis=AxisY, angularLimitsMax=<0 33 0>, angularLimitsMin=<0 -33 0>, angularTarget=<0 0 0>}
      localJointOffset=<-4.5 0 -5>
     }]
   }
   n7 = Anim Dynamics - Bone: dyn_legB_r_jnt(ComponentPose=@n6, Alpha=0.75, ExternalForce="0,0,0") {
    angularDampingOverride=150
    angularSpringConstant=5
    bGravityOverrideInSimSpace=true
    bUseGravityOverride=true
    alphaBoolBlend {blendOption=Linear}
    boundBone {boneName=dyn_legB_r_jnt}
    componentAppliedLinearAccClamp=<300 300 300>
    componentLinearAccScale=<1 1 1>
    componentLinearVelScale=<1 1 1>
    gravityOverride {z=-981}
    physicsBodyDefinitions [{
      sphereCollisionRadius=0
      boundBone {boneName=dyn_legB_r_jnt}
      boxExtents=<5 8 8>
      constraintSetup {angularTargetAxis=AxisY, angularLimitsMax=<0 22 12>, angularLimitsMin=<0 -33 -33>, angularTarget=<0 0 0>}
      localJointOffset=<2 0 -2>
     }]
   }
   n8 = Anim Dynamics - Bone: dyn_legC_r_jnt(ComponentPose=@n7, Alpha=0.5, ExternalForce="0,0,0") {
    angularDampingOverride=1200
    angularSpringConstant=10
    bGravityOverrideInSimSpace=true
    bUseGravityOverride=true
    alphaBoolBlend {blendOption=Linear}
    boundBone {boneName=dyn_legC_r_jnt}
    componentAppliedLinearAccClamp=<300 300 300>
    componentLinearAccScale=<1 1 1>
    componentLinearVelScale=<1 1 1>
    gravityOverride {z=-981}
    physicsBodyDefinitions [{
      sphereCollisionRadius=0
      boundBone {boneName=dyn_legC_r_jnt}
      boxExtents=<3 8 8>
      constraintSetup {angularTargetAxis=AxisY, angularLimitsMax=<0 30 12>, angularLimitsMin=<0 -30 -12>, angularTarget=<0 0 0>}
      localJointOffset=<2 0 -1>
     }]
   }
   n9 = Anim Dynamics - Bone: dyn_legB_l_jnt(ComponentPose=@n8, Alpha=0.75, ExternalForce="0,0,0") {
    angularDampingOverride=150
    angularSpringConstant=5
    bGravityOverrideInSimSpace=true
    bUseGravityOverride=true
    alphaBoolBlend {blendOption=Linear}
    boundBone {boneName=dyn_legB_l_jnt}
    componentAppliedLinearAccClamp=<300 300 300>
    componentLinearAccScale=<1 1 1>
    componentLinearVelScale=<1 1 1>
    gravityOverride {z=-981}
    physicsBodyDefinitions [{
      sphereCollisionRadius=0
      boundBone {boneName=dyn_legB_l_jnt}
      boxExtents=<5 8 8>
      constraintSetup {angularTargetAxis=AxisY, angularLimitsMax=<0 30 33>, angularLimitsMin=<0 -30 14>, angularTarget=<0 0 0>}
      localJointOffset=<5 0 -1>
     }]
   }
   n10 = Anim Dynamics - Bone: dyn_legC_l_jnt(ComponentPose=@n9, Alpha=0.5, ExternalForce="0,0,0") {
    angularDampingOverride=1200
    angularSpringConstant=10
    bGravityOverrideInSimSpace=true
    bUseGravityOverride=true
    alphaBoolBlend {blendOption=Linear}
    boundBone {boneName=dyn_legC_l_jnt}
    componentAppliedLinearAccClamp=<300 300 300>
    componentLinearAccScale=<1 1 1>
    componentLinearVelScale=<1 1 1>
    gravityOverride {z=-981}
    physicsBodyDefinitions [{
      sphereCollisionRadius=0
      boundBone {boneName=dyn_legC_l_jnt}
      boxExtents=<4 8 6>
      constraintSetup {angularTargetAxis=AxisY, angularLimitsMax=<0 30 12>, angularLimitsMin=<0 -30 -12>, angularTarget=<0 0 0>}
      localJointOffset=<2 0 -1>
     }]
   }
   n11 = Anim Dynamics - Bone: dyn_headA_jnt(ComponentPose=@n10, Alpha=0.5, ExternalForce="0,0,0") {
    angularDampingOverride=1200
    angularSpringConstant=10
    bGravityOverrideInSimSpace=true
    bUseGravityOverride=true
    alphaBoolBlend {blendOption=Linear}
    boundBone {boneName=dyn_headA_jnt}
    componentAppliedLinearAccClamp=<300 300 300>
    componentLinearAccScale=<1 1 1>
    componentLinearVelScale=<1 1 1>
    gravityOverride {z=-981}
    physicsBodyDefinitions [{
      sphereCollisionRadius=0
      boundBone {boneName=dyn_headA_jnt}
      boxExtents=<2 8 8>
      constraintSetup {angularTargetAxis=AxisZ, angularLimitsMax=<0 70.8344 60>, angularLimitsMin=<0 -98.3504 -8>, angularTarget=<0 0 0>}
      localJointOffset=<-1 0 -1>
     }]
   }
   n12 = Anim Dynamics - Bone: dyn_loBeak_jnt(ComponentPose=@n11, Alpha=0.5, ExternalForce="0,0,0") {
    angularDampingOverride=1200
    angularSpringConstant=10
    bGravityOverrideInSimSpace=true
    bUseGravityOverride=true
    alphaBoolBlend {blendOption=Linear}
    boundBone {boneName=dyn_loBeak_jnt}
    componentAppliedLinearAccClamp=<300 300 300>
    componentLinearAccScale=<1 1 1>
    componentLinearVelScale=<1 1 1>
    gravityOverride {z=-981}
    physicsBodyDefinitions [{
      sphereCollisionRadius=0
      boundBone {boneName=dyn_loBeak_jnt}
      boxExtents=<8 8 2>
      constraintSetup {angularTargetAxis=AxisY, angularLimitsMax=<0 60 12>, angularLimitsMin=<0 5 -12>, angularTarget=<0 0 0>}
      localJointOffset=<1 0 -2>
     }]
   }
   n13 = Copy Bone Delta - Source Bone: dyn_loBeak_jnt - Target Bone: dyn_upBeak_jnt!(ComponentPose:Component Space Pose Link Structure=@n12, Alpha:float=1) -> (Pose:Component Space Pose Link Structure) {
    is=title AnimGraphNode_CopyBoneDelta
    bCopyRotation=true
    rotationMultiplier=-1
    alphaBoolBlend {blendOption=Linear}
    sourceBone {boneName=dyn_loBeak_jnt}
    targetBone {boneName=dyn_upBeak_jnt}
   }
   n14 = Anim Dynamics - Bone: dyn_hat_jnt(ComponentPose=@n13, Alpha=0.5, ExternalForce="0,0,0") {
    angularDampingOverride=5
    bGravityOverrideInSimSpace=true
    bUseGravityOverride=true
    alphaBoolBlend {blendOption=Linear}
    boundBone {boneName=dyn_hat_jnt}
    componentAppliedLinearAccClamp=<300 300 300>
    componentLinearAccScale=<1 1 1>
    componentLinearVelScale=<1 1 1>
    gravityOverride {z=-9.81}
    physicsBodyDefinitions [{
      sphereCollisionRadius=0
      boundBone {boneName=dyn_hat_jnt}
      boxExtents=<4 1 4>
      constraintSetup {angularTargetAxis=AxisY, twistAxis=AxisY, angularLimitsMax=<39.516 0 0>, angularLimitsMin=<-36.0764 0 0>, angularTarget=<0 0 0>}
      localJointOffset=<-2 -3 -2>
     }]
   }
   n15 = Anim Dynamics - Bone: dyn_eye_r_jnt(ComponentPose=@n14, Alpha=0.5, ExternalForce="0,0,0") {
    angularDampingOverride=5
    angularSpringConstant=22
    bGravityOverrideInSimSpace=true
    bUseGravityOverride=true
    alphaBoolBlend {blendOption=Linear}
    boundBone {boneName=dyn_eye_r_jnt}
    componentAppliedLinearAccClamp=<300 300 300>
    componentLinearAccScale=<1 1 1>
    componentLinearVelScale=<1 1 1>
    gravityOverride {z=9.81}
    physicsBodyDefinitions [{
      sphereCollisionRadius=0
      boundBone {boneName=dyn_eye_r_jnt}
      boxExtents=<4 4 1>
      constraintSetup {angularTargetAxis=AxisY, twistAxis=AxisY, angularLimitsMax=<6 0 0>, angularLimitsMin=<-6 0 0>, angularTarget=<0 0 0>}
      localJointOffset=<0 0 1>
     }]
   }
   n16 = Anim Dynamics - Bone: dyn_eye_l_jnt(ComponentPose=@n15, Alpha=0.5, ExternalForce="0,0,0") {
    angularDampingOverride=5
    angularSpringConstant=22
    bGravityOverrideInSimSpace=true
    bUseGravityOverride=true
    alphaBoolBlend {blendOption=Linear}
    boundBone {boneName=dyn_eye_l_jnt}
    componentAppliedLinearAccClamp=<300 300 300>
    componentLinearAccScale=<1 1 1>
    componentLinearVelScale=<1 1 1>
    gravityOverride {z=9.81}
    physicsBodyDefinitions [{
      sphereCollisionRadius=0
      boundBone {boneName=dyn_eye_l_jnt}
      boxExtents=<4 4 1>
      constraintSetup {angularTargetAxis=AxisY, twistAxis=AxisY, angularLimitsMax=<6 0 0>, angularLimitsMin=<-6 0 0>, angularTarget=<0 0 0>}
      localJointOffset=<0 0 1>
     }]
   }
   n17 = "Transform (Modify) Bone - Bone: body_jnt"(ComponentPose=@n16, Alpha=Get spine_alpha() -> (spine_alpha:double), Translation="-0.000,-1.5,-0.000", Rotation="0,10,0", Scale="1,1,1") {
    rotationMode=BMM_Additive
    rotationSpace=BCS_BoneSpace
    translationMode=BMM_Additive
    translationSpace=BCS_BoneSpace
    alphaBoolBlend {blendOption=Linear}
    boneToModify {boneName=body_jnt}
    rotation {yaw=10}
    translation {y=-1.5}
   }
   n18 = "Transform (Modify) Bone - Bone: backpack_root"(ComponentPose=@n17, Alpha=0.25, Translation="0.000,0.000,8.145", Rotation=Get backpack_rotate(), Scale="1,1,1") {
    rotationMode=BMM_Additive
    rotationSpace=BCS_BoneSpace
    translationMode=BMM_Additive
    alphaBoolBlend {blendOption=Linear}
    boneToModify {boneName=backpack_root}
    translation {z=8.145}
   }
   Output Pose(Result=Component To Local(ComponentPose=@n18)) {layerGroup=None}
  ]}
 EventGraph {nodes [
   Event BlueprintUpdateAnimation()
   n1 = GetOwnerSkeletalMeshComponent()
   n2 = BreakRotator(InRot=BreakTransform(InTransform=GetDeltaTransformFromRefPose(self=@n1, BoneName=spine_05, BaseName=spine_01)).Rotation)
   > Set backpack_rotate(backpack_rotate=MakeRotator(Yaw=MapRangeClamped(Value=@n2.Yaw, InRangeA=-5, InRangeB=-30, OutRangeB=40)))
   > Set spine_alpha(spine_alpha:double=MapRangeClamped(Value=@n2.Yaw, InRangeA=0, InRangeB=34)) -> (Output_Get:double)
  ]}
]
variables [
 backpack_rotate {type=Rotator}
 spine_alpha {type=number}
]
